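; Top-score scoreboard actor for the Papaya Trials minigame, showing the best time and player name on a widget.
; It destroys itself once playlist data is ready if its enabled row is off.
; UpdateTopScore stores a submitted score whose time is equal or lower.
; The struct's OnRep caches the widget, formats minutes:seconds, substitutes a placeholder name when names are disabled, and plays celebration FX and fireworks sound unless it's the initial set.
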(event EventBeginPlay
  (Class|FortAsyncActionCurrentPlaylistReady|PlaylistDataReadyAsync
    (:OnReady
      (if (not (Class|AthenaFunctionLibrary|RowtoBool (Variables|Default|GetEnabledRow)))
        (Actor|DestroyActor)))))

(event Custom|UpdateTopScore
  (params (Struct_TopScore :type Struct%20Papaya%20Trials%20Top%20Score%20Structure))
  (if (<= (Utilities|Struct|BreakStructPapayaTrialsTopScore Struct_TopScore) (Utilities|Struct|BreakStructPapayaTrialsTopScore (Variables|Default|GetStruct_TopScore)))
    (Variables|Default|SetStruct_TopScore Struct_TopScore)))

(fn OnRep_Struct_TopScore
  (bind _struct_topscore (Variables|Default|GetStruct_TopScore))
  (bind (_time_2_f255704c4bb208a1c0cdd9b95e48ff92 _playername_6_4be4ff3d43c4d3d7d88aeeaa396257d7 _initialset_9_1e476b67494433b1c4b7f696ce10a39a) (Utilities|Struct|BreakStructPapayaTrialsTopScore _struct_topscore))
  (bind (_remainder _returnvalue) (Math|Float|Division(WholeandRemainder) _time_2_f255704c4bb208a1c0cdd9b95e48ff92 6e+01))
  (bind _widgetuserobject (Variables|Default|GetWidgetUserObject))
  (Utilities|IsValid _widgetuserobject
    (:Is%20Valid
      _)
    (:Is%20Not%20Valid
      (bind _aswidget_papaya_trials_top_score (Utilities|Casting|CastToWidget_Papaya_Trials_TopScore (UserInterface|GetUserWidgetObject (Variables|Default|GetWidget)))
        (:then
          (Variables|Default|SetWidgetUserObject _aswidget_papaya_trials_top_score))
        (:CastFailed
          (Development|PrintString "TIME TRIAL SCOREBOARD ERROR: Unable to find a valid User Widget" true true "(R=1.000000,G=0.056108,B=0.000000,A=1.000000)" 1e+01)))))
  (Class|WidgetPapayaTrialsTopScore|SetSCORETime _widgetuserobject (Utilities|Text|FormatText (Utilities|Text|MakeLiteralText "INVTEXT(\"{minutes}:{seconds}\")") (Utilities|Text|ToText(Float) _remainder "HalfToEven" false true 2 2 2 2) (Utilities|Text|ToText(Float) (select (< _returnvalue 0) 0.0 (Math|Conversions|ToFloat(Integer) _returnvalue)) "HalfToEven" false true 1 2 0 0)) _initialset_9_1e476b67494433b1c4b7f696ce10a39a (select (Class|AthenaFunctionLibrary|RowtoBool (Variables|Default|GetPlayerNameEnabled)) _playername_6_4be4ff3d43c4d3d7d88aeeaa396257d7 (Variables|Default|GetDisabledPlayerName)))
  (if (not _initialset_9_1e476b67494433b1c4b7f696ce10a39a)
    (Components|Activation|Activate (Variables|Default|GetFX_Celebration) true)
    (Audio|PlaySoundatLocation (Variables|Default|GetSFX_Fireworks) (Transformation|GetActorLocation))))

(fn ConstructionScript)
