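# Widget that shows the Papaya Trials top score time and record holder's name beneath a "Top Score" callout.
# SetSCORE_Time stores the time and name, then applies them immediately for an initial score or starts a clipped RollOut animation otherwise.
# Animation-finished events chain RollOut, RollIn1 and RollIn2, swapping the text after RollOut.
# The callout switcher shows "New Record!" during RollIn2, with clipping reset to Inherit, then returns to "Top Score".
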WidgetBlueprint /PapayaGameplay/Minigames/Trials/Widget_Papaya_Trials_TopScore
graphs [
 EventGraph {nodes [
   n1 = "On Animations Finished (AnimationsPanel_Score)"!() -> (OutputDelegate, then, Name:Name=None) {is=title K2Node_ComponentBoundEvent}
   > n2 = Branch(Condition="Equal (Name)"!(A:Name=@n1.Name, B:Name=Get RollOut() -> (RollOut:Name=None)) -> (ReturnValue:bool) {is=title K2Node_CallFunction})
   > UpdateTime()
   > PlayAnimationByName!(
    execute
    self:Fort Simple Widget Animations Panel Object Reference=Get AnimationsPanel_Score() -> (AnimationsPanel_Score:Fort Simple Widget Animations Panel Object Reference)
    Name:"Name (by ref)"=Get RollIn1() -> (RollIn1:Name=None)
    Speed:float=1
    StartAtTime:float=0
   ) -> (then, ReturnValue:bool)
   n3 = Branch(execute=@n2.else, Condition="Equal (Name)"(A=@n1.Name, B=Get RollIn1() -> (RollIn1=None)))
   > SetActiveWidgetIndex(self=Get CommonWidgetSwitcher_Callout() -> (CommonWidgetSwitcher_Callout:Common Widget Switcher Legacy Object Reference), Index=1)
   > PlayAnimationByName(self=Get AnimationsPanel_Score(), Name=Get RollIn2() -> (RollIn2:Name=None), Speed=1, StartAtTime=0)
   > SetClipping!(
    execute
    self:Widget Object Reference=Get Border_Score() -> (Border_Score:Common Border Object Reference)
    InClipping:EWidgetClipping Enum=Inherit
   ) -> (then)
   Branch(execute=@n3.else, Condition="Equal (Name)"(A=@n1.Name, B=Get RollIn2() -> (RollIn2=None)))
   > SetActiveWidgetIndex(self=Get CommonWidgetSwitcher_Callout(), Index=0)
   n4 = SetSCORE_Time!() -> (OutputDelegate, then, Text:Text, isInitialScore:bool=~, PlayerName:Text) {is=title K2Node_CustomEvent}
   > Set NewTIme(NewTIme:Text=@n4.Text) -> (Output_Get:Text)
   > Set PlayerName(PlayerName:Text=@n4.PlayerName) -> (Output_Get:Text)
   > n5 = Branch(Condition=@n4.isInitialScore)
   > "SetText (Text)"(self=Get SCORE_Time() -> (SCORE_Time:Common Text Object Reference), InText=Get NewTIme() -> (NewTIme:Text))
   > "SetText (Text)"(self=Get Name() -> (Name:Common Text Object Reference), InText=Get PlayerName() -> (PlayerName:Text))
   PlayAnimationByName(execute=@n5.else, self=Get AnimationsPanel_Score(), Name=Get RollOut() -> (RollOut=None), Speed=1, StartAtTime=0)
   > SetActiveWidgetIndex(self=Get CommonWidgetSwitcher_Callout(), Index=0)
   > SetClipping(self=Get Border_Score(), InClipping=ClipToBounds)
  ]}
 Sequencer Events {nodes [
   UpdateTime() {is=title K2Node_CustomEvent}
   > "SetText (Text)"(self=Get SCORE_Time(), InText=Get NewTIme())
   > "SetText (Text)"(self=Get Name(), InText=Get PlayerName())
  ]}
]
variables [
 NewTIme {type=string, default=" "}
 PlayerName {type=string, default=" "}
 RollIn1 {type=string, default=RollIn1}
 RollIn2 {type=string, default=RollIn2}
 RollOut {type=string, default=RollOut}
]
widgets {
 info {inheritedWidgetCount=0, parentClass=&UserWidget, rootWidgetClass=&SizeBox, widgetCount=12}
 widgets [
  SizeBox_1 {
   widgetClassPath=&SizeBox
   widget=SizeBox_1
   properties {bOverride_HeightOverride=true, bOverride_WidthOverride=true, heightOverride=200, widthOverride=500}
  }
  Border_14 {
   widgetClassPath=&Border
   parent=SizeBox_1
   widget=Border_14
   properties {verticalAlignment=VAlign_Center, brushColor=<rgba 1 1 1 0>}
   slot {class=SizeBoxSlot}
  }
  VerticalBox_0 {
   widgetClassPath=&VerticalBox
   parent=Border_14
   widget=VerticalBox_0
   slot {class=BorderSlot, properties {verticalAlignment=VAlign_Center}}
  }
  SizeBox_0 {
   widgetClassPath=&SizeBox
   parent=VerticalBox_0
   widget=SizeBox_0
   properties {bOverride_WidthOverride=true}
   slot {class=VerticalBoxSlot, properties {padding=<ltrb 0 0 0 -8>}}
  }
  ScaleBox_0 {widgetClassPath=&ScaleBox, parent=SizeBox_0, widget=ScaleBox_0, slot {class=SizeBoxSlot}}
  CommonWidgetSwitcher_Callout {
   widgetClassPath=&CommonWidgetSwitcherLegacy
   bIsVariable=true
   parent=ScaleBox_0
   widget=CommonWidgetSwitcher_Callout
   properties {transitionCurveType=QuadIn, transitionType=Zoom}
   slot {class=ScaleBoxSlot}
  }
  TopScore {
   widgetClassPath=&CommonTextBlock
   bIsVariable=true
   parent=CommonWidgetSwitcher_Callout
   widget=TopScore
   properties {
    justification=Center
    text=Top Score
    textTransformPolicy=ToUpper
    visibility=Visible
    font {
     bForceMonospaced=false
     bMaterialIsStencil=false
     fontMaterial=null
     fontObject=&Burbank
     letterSpacing=0
     monospacedWidth=1
     size=26
     skewAmount=0
     typefaceFontName=BurbankRegular-Black
     outlineSettings {
      bApplyOutlineToDropShadows=false
      bMiteredCorners=false
      bSeparateFillAlpha=false
      outlineMaterial=null
      outlineSize=0
      outlineColor=<rgba 0 0 0 1>
     }
    }
    renderTransform {angle=0, scale=<1 1>, shear=<-8 0>, translation=<0 0>}
    shadowOffset=<0 0>
   }
   slot {class=WidgetSwitcherSlot, properties {horizontalAlignment=HAlign_Center, verticalAlignment=VAlign_Center}}
  }
  NewRecord {
   widgetClassPath=&CommonTextBlock
   parent=CommonWidgetSwitcher_Callout
   widget=NewRecord
   properties {
    justification=Center
    text="New Record!"
    textTransformPolicy=ToUpper
    visibility=Visible
    colorAndOpacity {colorUseRule=UseColor_Specified, specifiedColor=<rgba 1 0.7463 0 1>}
    font {
     bForceMonospaced=false
     bMaterialIsStencil=false
     fontMaterial=null
     fontObject=&Burbank
     letterSpacing=0
     monospacedWidth=1
     size=26
     skewAmount=0
     typefaceFontName=BurbankRegular-Black
     outlineSettings {
      bApplyOutlineToDropShadows=false
      bMiteredCorners=false
      bSeparateFillAlpha=false
      outlineMaterial=null
      outlineSize=0
      outlineColor=<rgba 0 0 0 1>
     }
    }
    renderTransform {angle=0, scale=<1 1>, shear=<-8 0>, translation=<0 0>}
    shadowOffset=<0 0>
   }
   slot {class=WidgetSwitcherSlot, properties {horizontalAlignment=HAlign_Center, verticalAlignment=VAlign_Center}}
  }
  AnimationsPanel_Score {
   widgetClassPath=/Script/FortniteUI.FortSimpleWidgetAnimationsPanel
   bIsVariable=true
   parent=VerticalBox_0
   widget=AnimationsPanel_Score
   properties {animations [
     RollOut {
      bIsEnabled=true
      onAnimationsFinished=()
      targets [{
        alpha=null
        bResetOnFinish=false
        bShouldAnimateAlpha=false
        bShouldAnimateScale=false
        bShouldAnimateTranslation=true
        scale=null
        targetName=SCORE_Time
        translation=&/Widget_Papaya_Trials_TopScore_RollOutTranslation
        materialScalarAnimations []
       }]
     }
     RollIn1 {
      bIsEnabled=true
      onAnimationsFinished=()
      targets [{
        alpha=null
        bResetOnFinish=false
        bShouldAnimateAlpha=false
        bShouldAnimateScale=false
        bShouldAnimateTranslation=true
        scale=null
        targetName=SCORE_Time
        translation=&/Widget_Papaya_Trials_TopScore_RollIn1Translation
        materialScalarAnimations []
       }]
     }
     RollIn2 {
      bIsEnabled=true
      onAnimationsFinished=()
      targets [{
        alpha=null
        bResetOnFinish=false
        bShouldAnimateAlpha=false
        bShouldAnimateScale=false
        bShouldAnimateTranslation=true
        scale=null
        targetName=SCORE_Time
        translation=&/Widget_Papaya_Trials_TopScore_RollIn2Translation
        materialScalarAnimations []
       }]
     }
    ]}
   slot {class=VerticalBoxSlot}
  }
  Border_Score {
   widgetClassPath=&CommonBorder
   bIsVariable=true
   parent=AnimationsPanel_Score
   widget=Border_Score
   properties {verticalAlignment=VAlign_Center, padding=<ltrb 0 -8 0 -12>}
   slot {class=FortSimpleWidgetAnimationsSlot}
  }
  SCORE_Time {
   widgetClassPath=&CommonTextBlock
   bIsVariable=true
   parent=Border_Score
   widget=SCORE_Time
   properties {
    justification=Center
    text=0:00.00
    visibility=Visible
    font {
     bForceMonospaced=false
     bMaterialIsStencil=false
     fontMaterial=null
     fontObject=&Burbank
     letterSpacing=0
     monospacedWidth=1
     size=40
     skewAmount=0
     typefaceFontName=Burbank-PlayerPower
     outlineSettings {
      bApplyOutlineToDropShadows=false
      bMiteredCorners=false
      bSeparateFillAlpha=false
      outlineMaterial=null
      outlineSize=0
      outlineColor=<rgba 0 0 0 1>
     }
    }
    shadowOffset=<0 0>
   }
   slot {class=BorderSlot, properties {verticalAlignment=VAlign_Center, padding=<ltrb 0 -8 0 -12>}}
  }
  Name {
   widgetClassPath=&CommonTextBlock
   bIsVariable=true
   parent=VerticalBox_0
   widget=Name
   properties {
    justification=Center
    text=YourNameHere
    visibility=Visible
    font {
     bForceMonospaced=false
     bMaterialIsStencil=false
     fontMaterial=null
     fontObject=&Burbank
     letterSpacing=0
     monospacedWidth=1
     size=22
     skewAmount=0
     typefaceFontName=BurbankRegular-Black
     outlineSettings {
      bApplyOutlineToDropShadows=false
      bMiteredCorners=false
      bSeparateFillAlpha=false
      outlineMaterial=null
      outlineSize=0
      outlineColor=<rgba 0 0 0 1>
     }
    }
    renderTransform {angle=0, scale=<1 1>, shear=<-8 0>, translation=<0 0>}
    shadowOffset=<0 0>
   }
   slot {class=VerticalBoxSlot, properties {horizontalAlignment=HAlign_Center, verticalAlignment=VAlign_Center}}
  }
 ]
}
